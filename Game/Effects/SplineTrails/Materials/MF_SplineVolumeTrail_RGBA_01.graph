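# Material function for BP_SplineVolumeTrail effects that lets the Blueprint drive RGB, alpha and scale WPO, plus a SplineID pass-through.
# With CheapSplines on, it samples separate rows of a color-curve texture at two segment ends, using timestamp age over lifetime or manual values, and blends along U or V.
# Curve values also scale offsets toward WP_A and WP_B; otherwise color parameters are blended and WPO is zero.
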MaterialFunction /Game/Effects/SplineTrails/Materials/MF_SplineVolumeTrail_RGBA_01
graph {
 properties {bExposeToLibrary=true, description=For use with BP_SplineVolumeTrail_x, libraryCategoriesText [VFX]}
 nodes [
  n1 = Time() {period=1}
  n2 = Subtract(A=@n1, B=ScalarParameter() {group=Timestamp, parameterName=SplineTimestampA})
  n3 = ScalarParameter() {defaultValue=1, group=Timestamp, parameterName=SplineMeshLifetime}
  n4 = FunctionInput(Preview=StaticBoolParameter() {parameterName=ColorOverLengthOfSpline}) {bUsePreviewValueAsDefault=true, inputName=ColorOverLengthOfSpline, inputType=FunctionInput_StaticBool, sortPriority=2}
  n5 = StaticSwitch(True=ScalarParameter() {parameterName=ManualA}, False=Saturate(None=Divide(A=@n2, B=@n3)), Value=@n4)
  n6 = ScalarParameter() {defaultValue=0.125, parameterName=V_ColorRGB}
  n7 = TextureSampleParameter2D(UVs=AppendVector(A=@n5, B=@n6)) {
   desc=A
   parameterName=LinearColorTexture
   samplerType=SAMPLERTYPE_LinearColor
   texture=/Game/Effects/SplineTrails/Textures/T_LinearColorCurve_Default_01
  }
  n8 = Subtract(A=@n1, B=ScalarParameter() {defaultValue=9999999827968, group=Timestamp, parameterName=SplineTimestampB})
  n9 = StaticSwitch(True=ScalarParameter() {defaultValue=1, parameterName=ManualB}, False=Saturate(None=Divide(A=@n8, B=@n3)), Value=@n4)
  n10 = TextureSampleParameter2D(UVs=AppendVector(A=@n9, B=@n6)) {
   desc=B
   parameterName=LinearColorTexture
   samplerType=SAMPLERTYPE_LinearColor
   texture=&T_LinearColorCurve_Default_01
  }
  n11 = TextureCoordinate()
  n12 = StaticSwitch(
   True=ComponentMask(None=@n11) {g=true}
   False=ComponentMask(None=@n11) {r=true}
   Value=FunctionInput(Preview=StaticBoolParameter() {defaultValue=true, parameterName=V}) {bUsePreviewValueAsDefault=true, inputName=Use V Coord, inputType=FunctionInput_StaticBool}
  )
  n13 = VectorParameter() {parameterName=LinearColorB, defaultValue=<rgba 1 1 1 1>}
  n14 = VectorParameter() {parameterName=LinearColorA, defaultValue=<rgba 1 1 1 1>}
  n15 = FunctionInput(Preview=StaticBoolParameter() {defaultValue=true, parameterName=CheapSplines}) {bUsePreviewValueAsDefault=true, inputName=Cheap Splines, inputType=FunctionInput_StaticBool, sortPriority=1}
  FunctionOutput(None=StaticSwitch(
    True=LinearInterpolate(A=@n7.RGB, B=@n10.RGB, Alpha=@n12) {constB=0}
    False=LinearInterpolate(A=@n13.RGB, B=@n14.RGB, Alpha=@n12)
    Value=@n15
   )) {description=For use with BP_SplineVolumeTrail - quick setup to allow the bp to drive RGBA values for color and alpha., outputName=RGB}
  FunctionOutput(None=ScalarParameter() {parameterName=SplineID}) {outputName=SplineID, sortPriority=99}
  n16 = ScalarParameter() {defaultValue=0.375, parameterName=V_ColorA}
  n17 = TextureSampleParameter2D(UVs=AppendVector(A=@n5, B=@n16)) {
   desc=A
   parameterName=LinearColorTexture
   samplerType=SAMPLERTYPE_LinearColor
   texture=&T_LinearColorCurve_Default_01
  }
  n18 = TextureSampleParameter2D(UVs=AppendVector(A=@n9, B=@n16)) {
   desc=B
   parameterName=LinearColorTexture
   samplerType=SAMPLERTYPE_LinearColor
   texture=&T_LinearColorCurve_Default_01
  }
  n19 = VectorParameter() {parameterName=LinearColorB, defaultValue=<rgba 1 1 1 1>}
  n20 = VectorParameter() {parameterName=LinearColorA, defaultValue=<rgba 1 1 1 1>}
  FunctionOutput(None=StaticSwitch(True=LinearInterpolate(A=@n17.RGB, B=@n18.RGB, Alpha=@n12) {constB=0}, False=LinearInterpolate(A=@n19.A, B=@n20.A, Alpha=@n12), Value=@n15)) {outputName=A, sortPriority=1}
  n21 = ScalarParameter() {defaultValue=0.625, parameterName=V_ColorWPO}
  n22 = TextureSampleParameter2D(UVs=AppendVector(A=@n5, B=@n21)) {
   desc=A
   parameterName=LinearColorTexture
   samplerType=SAMPLERTYPE_LinearColor
   texture=&T_LinearColorCurve_Default_01
  }
  n23 = LinearInterpolate(A=Subtract(A=VectorParameter() {parameterName=WP_A, defaultValue=<rgba 100 0 0 0>}.RGB, B=WorldPosition().XYZ), Alpha=@n22.RGB) {constB=0}
  n24 = TextureSampleParameter2D(UVs=AppendVector(A=@n9, B=@n21)) {
   desc=B
   parameterName=LinearColorTexture
   samplerType=SAMPLERTYPE_LinearColor
   texture=&T_LinearColorCurve_Default_01
  }
  n25 = LinearInterpolate(A=Subtract(A=VectorParameter() {parameterName=WP_B, defaultValue=<rgba -100 0 0 0>}.RGB, B=WorldPosition().XYZ), Alpha=@n24.RGB) {constB=0}
  FunctionOutput(None=StaticSwitch(True=LinearInterpolate(A=@n23, B=@n25, Alpha=@n12) {constB=0}, False=Constant(), Value=@n15)) {outputName=Scale WPO, sortPriority=2}
  MaterialFunctionCall!(In) {materialFunction=/Game/Effects/Fort_Effects/Materials/_Functions/MF_Smoothstep_01}
 ]
}
