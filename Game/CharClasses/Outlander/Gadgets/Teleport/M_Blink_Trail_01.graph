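# Translucent, unlit, two-sided spline-mesh material for the trail of the Outlander's Teleport (Blink) gadget.
# Two opposite-panning plasma noise samples, mapped by view-space normals, blend two colors whose hues oscillate with time and fresnel.
# A spline volume trail function modulates color and alpha along the length.
# Noise, fresnel and vertex color shape opacity, which fades with distance and, with a UV1 cosine edge mask, drives index-of-refraction distortion.
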Material /Game/CharClasses/Outlander/Gadgets/Teleport/M_Blink_Trail_01
graph {
 properties {
  bEnableResponsiveAA=true
  bUsedWithSplineMeshes=true
  blendMode=BLEND_Translucent
  pixelDepthOffsetMode=PDOM_Legacy
  previewMesh=/Game/Effects/SplineTrails/Meshes/Sm_Blink_Trail_01
  refractionMethod=RM_IndexOfRefraction
  shadingModel=MSM_Unlit
  twoSided=true
  usedShadingModels=Unlit
 }
 nodes [
  n1 = Multiply(A=Time() {period=3}) {constB=0.3}
  n2 = Multiply(
   A=Transform(None=PixelNormalWS()) {transformSourceType=TRANSFORMSOURCE_World, transformType=TRANSFORM_View}
   B=ScalarParameter() {defaultValue=0.8, parameterName=Normal Strength}
  )
  Add(
   A=Add(A=CameraVectorWS(), B=MaterialFunctionCall(G=@n1) {materialFunction=&AppendMany}.RGB)
   B=@n2
  )
  n3 = Multiply(A=Sine(None=Time()) {period=2.955}) {constB=0.12}
  n4 = MaterialFunctionCall(
   Brightness=ScalarParameter() {defaultValue=1.05, parameterName=Fresnel Brightness}
   Exponent=ScalarParameter() {defaultValue=1.15, parameterName=Fresnel Exponent}
  ) {materialFunction=&MF_Fresnel_FX}
  n5 = Multiply(A=ScalarParameter() {defaultValue=2.2, parameterName=Fresnel Hue Shift Intensity}, B=@n4.Fresnel) {constA=2}
  n6 = Multiply(A=Sine(None=Time()) {period=2.1}) {constB=0.25}
  n7 = Multiply(A=VertexColor().R, B=ScalarParameter() {defaultValue=0.5, parameterName=Vert Color Edge})
  n8 = StaticComponentMaskParameter(None=@n2) {defaultG=true, defaultR=true, parameterName=Mask}
  n9 = TextureSample(UVs=Panner(Coordinate=@n8) {bFractionalPart=true, speedX=0.1875, speedY=-0.142}) {texture=/Game/Effects/Fort_Effects/Textures/Plasma/T_PlasmaNoise_A_01}
  n10 = Multiply(
   A=@n9.G
   B=TextureSample(UVs=Panner(Coordinate=@n8) {bFractionalPart=true, speedX=-0.262, speedY=0.176}) {texture=&T_PlasmaNoise_A_01}.R
  )
  n11 = LinearInterpolate(
   A=MaterialFunctionCall(Hue Shift Percentage=Multiply(A=@n3, B=@n5), Texture=VectorParameter() {parameterName=Color1, defaultValue=<rgba 0.0048 0.0055 0.06 0>}.RGB) {materialFunction=&HueShift}.Result
   B=MaterialFunctionCall(Hue Shift Percentage=Multiply(A=@n6, B=@n5), Texture=VectorParameter() {parameterName=Color2, defaultValue=<rgba 0.42 0.0553 0.3686 0>}.RGB) {materialFunction=&HueShift}.Result
   Alpha=Clamp(None=Add(A=@n7, B=@n10))
  )
  n12 = Power(Base=ConstantBiasScale(None=Multiply(A=@n9.G, B=@n10)) {bias=0.5, scale=2}, Exp=ScalarParameter() {defaultValue=4.5, parameterName=Exponent}) {constExponent=4}
  n13 = StaticBool() {value=true}
  n14 = MaterialFunctionCall!(Use V Coord, Cheap Splines=@n13, ColorOverLengthOfSpline=@n13) {materialFunction=/Game/Effects/SplineTrails/Materials/MF_SplineVolumeTrail_RGBA_01}
  Reroute(None=Multiply(A=Multiply(A=@n11, B=@n12), B=@n14.RGB))
  n15 = ScalarParameter() {defaultValue=0.9, parameterName=Refraction}
  n16 = ComponentMask(None=TextureCoordinate() {coordinateIndex=1, uTiling=0.5, vTiling=0.5}) {g=true}
  n17 = Abs(None=Cosine(None=@n16))
  n18 = Saturate(None=ConstantBiasScale(None=OneMinus(None=@n17)) {bias=-0.1, scale=1.2})
  n19 = ScalarParameter() {defaultValue=0.55, parameterName=Opacity Darks}
  n20 = ScalarParameter() {defaultValue=1, parameterName=Opacity Lights}
  n21 = Add(A=@n7, B=Multiply(A=LinearInterpolate(A=@n19, B=@n20, Alpha=@n10), B=@n4.Fresnel))
  n22 = MaterialFunctionCall(
   Opacity In=Clamp(None=Multiply(A=@n21, B=@n14.A))
   Position=WorldPosition().XYZ
   Fade_Start_Distance=ScalarParameter() {defaultValue=200, parameterName=Start Fade Distance}
   Fade_End_Distance=ScalarParameter() {defaultValue=50, parameterName=Start End Distance}
  ) {materialFunction=&MF_Distance_SyncedFade_FX}
  Reroute(None=LinearInterpolate(B=LinearInterpolate(B=@n15, Alpha=@n4.Fresnel) {constA=1}, Alpha=Multiply(A=@n18, B=@n22.Result)) {constA=1})
  n23 = ComponentMask(None=TextureCoordinate() {uTiling=0.5}) {r=true}
  n24 = ScalarParameter() {defaultValue=1, parameterName=S1}
  Sine(None=Add(A=Multiply(A=@n23, B=@n24), B=ScalarParameter() {defaultValue=0.25, parameterName=S2}))
  Sine(None=@n16)
 ]
}
