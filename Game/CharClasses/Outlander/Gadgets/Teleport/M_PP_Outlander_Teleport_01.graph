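# Post-process material for the Outlander teleport gadget's screen effect, blended in by the OutlanderTeleportPP collection parameter.
# Scene UVs are zoomed about screen center and warped by a spherical normal map.
# Two counter-panning light-ray samples in radial coordinates form bright streaks, added to a vignette color, hue-shifted and multiplied over the scene outside a central SphereMask.
# Leftover screen-resolution debug nodes feed nothing in the effect chain.
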Material /Game/CharClasses/Outlander/Gadgets/Teleport/M_PP_Outlander_Teleport_01
graph {
 properties {materialDomain=MD_PostProcess, pixelDepthOffsetMode=PDOM_Legacy, usedShadingModels=Default Lit}
 nodes [
  n1 = MaterialFunctionCall() {materialFunction=&ScreenResolution}
  n2 = Constant() {r=4}
  n3 = TextureCoordinate() {uTiling=1.5}
  n4 = MaterialFunctionCall!(Vector2=@n1.Visible Resolution, MaximumNumberOfDigits=@n2, UVs=@n3, DebugTextLocation RG_UpperRight BA_LowerLeft, Component Spacing) {materialFunction=/Engine/Functions/Engine_MaterialFunctions02/Utility/DebugFloat2Values}
  n5 = MaterialFunctionCall(Vector2=@n1.Buffer Resolution, MaximumNumberOfDigits=@n2, UVs=Add(A=@n3, B=Constant2Vector() {r=-0.5})) {materialFunction=&DebugFloat2Values}
  Add(B=AppendVector(A=Add(A=@n4.ColorCodedOutput, B=@n5.ColorCodedOutput), B=Constant()))
  Constant() {r=1}
  n6 = ScreenPosition()
  n7 = ScalarParameter() {defaultValue=0.9, parameterName=Screen Scale}
  n8 = Multiply(A=@n6.ViewportUV, B=@n7)
  n9 = Divide(A=OneMinus(None=@n7))
  n10 = ComponentMask(None=TextureSample() {samplerType=SAMPLERTYPE_Normal, texture=/Game/Gadgets/Assets/Teleporter/Textures/T_Spherical_Normal_02}.RGB) {g=true, r=true}
  n11 = ScalarParameter() {defaultValue=0.04, parameterName=Intensity}
  n12 = CollectionParameter() {collection=&FortniteMaterialParameters, parameterName=OutlanderTeleportPP}
  n13 = LinearInterpolate(A=@n6.ViewportUV, B=Add(A=Add(A=@n8, B=@n9), B=Multiply(A=@n10, B=@n11)), Alpha=@n12)
  n14 = ComponentMask(None=SceneTexture(UVs=@n13) {sceneTextureId=PPI_PostProcessInput0}.Color) {b=true, g=true, r=true}
  n15 = MaterialFunctionCall() {materialFunction=&VectorToRadialValue}
  n16 = Multiply(
   A=TextureSample(UVs=Panner(Coordinate=@n15.Radial Coordinates) {speedX=0.3248}) {texture=/Game/Effects/Fort_Effects/Textures/Outlander/T_InTheZone_LightRays_01}.R
   B=TextureSample(UVs=Panner(Coordinate=@n15.Radial Coordinates) {speedX=-0.484}) {texture=&T_InTheZone_LightRays_01}.G
  )
  n17 = OneMinus(None=@n16)
  n18 = Multiply(A=Power(Base=Multiply(A=@n16, B=@n17)) {constExponent=8}, B=ScalarParameter() {defaultValue=75000, parameterName=Line Brightness}) {constB=3000}
  n19 = MaterialFunctionCall(
   Hue Shift Percentage=ConstantBiasScale(None=@n16) {bias=-0.5, scale=0.4}
   Texture=Add(A=VectorParameter() {parameterName=Vingette Color, defaultValue=<rgba 0.7591 0.3104 0.97 0>}.RGB, B=Clamp(None=@n18))
  ) {materialFunction=&HueShift}
  n20 = Constant() {r=0.5}
  n21 = ScalarParameter() {defaultValue=0.6, parameterName=Mask Radius}
  n22 = ScalarParameter() {defaultValue=0.4, parameterName=Mask Hardness}
  LinearInterpolate(
   A=@n14
   B=LinearInterpolate(A=Multiply(A=@n19.Result, B=@n14), B=@n14, Alpha=SphereMask(A=@n13, B=@n20, Radius=@n21, Hardness=@n22))
   Alpha=@n12
  )
 ]
}
